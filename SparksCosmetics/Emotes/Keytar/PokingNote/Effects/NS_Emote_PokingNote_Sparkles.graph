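# Niagara system for the sparkle effect of the Keytar PokingNote emote, built from a single sprite emitter.
# Sparkles spawn continuously in a small sphere with short random lifetimes and a color picked from a multicolor palette array.
# Random sprite size passes through a front-end/gameplay modifier, currently equal for both.
# Each sparkle pops in and shrinks out on a size curve, drifts under gentle downward gravity, and gets a strong RGB boost for glow.
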NiagaraSystem /SparksCosmetics/Emotes/Keytar/PokingNote/Effects/NS_Emote_PokingNote_Sparkles
dependencies {
 usedDataInterfaces [&NiagaraDataInterfaceArrayColor, &NiagaraDataInterfaceCurve]
 usedDynamicInputs []
 usedModules [&SystemState]
}
emitters {Sparkles {
  emitterUpdateScript [
   &EmitterState()
   &SpawnRate(SpawnRate=20, Use Spawn Probability=true)
  ]
  particleSpawnScript [
   &InitializeParticle(
    Color=&SelectLinearColorFromArray(
     Array Sampling Mode=::NewEnumerator0
     Direct Array Index=?
     Color Selection Array=DataInterface() {propertyValues {ColorData [<rgba 0.1042 0.1042 1 1>, <rgba 0 1 0.3229 1>, <rgba 0 0.9255 1 1>, <rgba 0.9546 0 0.5008 1>, <rgba 1 0 0.1741 1>, <rgba 1 0.8502 0.014 1>]}}
     Random Seed=0
    )
    Lifetime Max=0.8
    Lifetime Min=0.6
    Lifetime Mode=::NewEnumerator1
    Sprite Rotation Angle=0.125
    Sprite Rotation Mode=::NewEnumerator0
    Sprite Size Mode=::NewEnumerator3
    Uniform Sprite Size=&Multiply_Float(
     A=&RandomRangeFloat(Minimum=20, Maximum=30)
     B=&FN_FrontEndFloatModifier(Gameplay=1, Front End=1)
    )
    UsePositionOffset=false
    Write Material Random=true
   )
   &ShapeLocation(MaskSpawnGroup=false, Override Seed=false, Sphere Radius=5, UseOwnerScale=true)
  ]
  particleUpdateScript [
   &ParticleState()
   &ScaleSpriteSize(Scale Factor=&Vector2DFromFloat(Value=&FloatFromCurve(FloatCurve=DataInterface() {propertyValues {Curve {keys [<at 0 0>,<at 0.3 1>,<at 0.6 1.075>,<at 1 0>]}}})))
   &GravityForce(Gravity=<0 0 -30>)
   &SolveForcesAndVelocity(Clamp Velocity=false, Limit Acceleration=false)
   &ScaleColor(Scale RGB=&VectorFromFloat(Value=13), ScaleA=true, ScaleRGB=true)
  ]
  rendererClasses [&NiagaraSpriteRendererProperties]
  renderers [{rendererClass=&NiagaraSpriteRendererProperties}]
 }}
summary {
 systemName=NS_Emote_PokingNote_Sparkles
 emitters [{emitterName=Sparkles, rendererClasses [&NiagaraSpriteRendererProperties]}]
 userVariables [
  User.bIsFrontEnd {type=NiagaraBool}
  User.bIsFrontEndPreview {type=NiagaraBool}
 ]
}
